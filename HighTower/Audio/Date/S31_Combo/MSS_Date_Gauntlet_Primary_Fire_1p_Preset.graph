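# First-person primary fire one-shot MetaSound source for the Date Gauntlet weapon, in 5.1.
# On play, it picks a random close-fire wave with random pitch shift; the Fire input is unwired.
# ShotsLeft, compared against a decibel table's length, selects a table gain scaled by MasterVolume or resets to MasterVolume.
# A trigger sequence feeds two gain-scaled, summed players, with trigger gates so only the latest reports OnFinished.
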MetaSoundSource /HighTower/Audio/Date/S31_Combo/MSS_Date_Gauntlet_Primary_Fire_1p_Preset
inputs [
 Fire {type=Trigger}
 MasterVolume {type=Float}
 Random.PitchShift.Base {type=Float}
 Random.PitchShift.Spread {type=Float}
 ShotsLeft {type=Int32}
 Sounds {type=WaveAsset:Array}
 UE.Source.OnPlay {type=Trigger}
]
interfaces [UE.OutputFormat.5dot1:1.0, UE.Source.OneShot:1.0, UE.Source:1.1]
outputs [
 "UE.OutputFormat.5dot1.Audio:0" {type=Audio}
 "UE.OutputFormat.5dot1.Audio:1" {type=Audio}
 "UE.OutputFormat.5dot1.Audio:2" {type=Audio}
 "UE.OutputFormat.5dot1.Audio:3" {type=Audio}
 "UE.OutputFormat.5dot1.Audio:4" {type=Audio}
 "UE.OutputFormat.5dot1.Audio:5" {type=Audio}
 UE.Source.OneShot.OnFinished {type=Trigger}
]
pages [Default {
  input_defaults {
   Fire=null
   MasterVolume=1
   Random.PitchShift.Base=0
   Random.PitchShift.Spread=0.5
   ShotsLeft=5
   UE.Source.OnPlay=null
   Sounds [
    &/Date_Gauntlet_Primary_Fire_1P_Close_01
    &/Date_Gauntlet_Primary_Fire_1P_Close_02
    &/Date_Gauntlet_Primary_Fire_1P_Close_03
    &/Date_Gauntlet_Primary_Fire_1P_Close_04
   ]
  }
  nodes [
   n1 = VariableAccessor.Float:Array(Variable=InitVariable.Float:Array().Variable)
   ArrayRandomGet_0 = Array.Random Get:WaveAsset:Array(In Array=^Sounds, Next=^UE.Source.OnPlay) {also {defaults {Weights []}}}
   n2 = TriggerCompare.Clamp:Int32(A=^ShotsLeft, B=Array.Num:Float:Array(Array=@n1.Value).Num, Compare=@ArrayRandomGet_0.On Next) {also {defaults {Type=5}}}
   DecibelstoLinearGain_0 = UE.Decibels to Linear Gain:Float(Decibels=Array.Get:Float:Array(Index=^ShotsLeft, Trigger=@n2.False, Array=VariableAccessor.Float:Array(Variable=@n1.Variable).Value).Element)
   n3 = UE.Multiply:Float(PrimaryOperand=^MasterVolume, AdditionalOperands=@DecibelstoLinearGain_0.Linear Gain)
   n4 = UE.RandomFloat(
    Max=UE.Add:Float(PrimaryOperand=^Random.PitchShift.Base, AdditionalOperands=^Random.PitchShift.Spread).Out
    Next=@ArrayRandomGet_0.On Next
    Min=UE.Subtract:Float(PrimaryOperand=^Random.PitchShift.Base, AdditionalOperands=^Random.PitchShift.Spread).Out
   )
   TriggerSequence2_0 = "UE.Trigger Sequence (2)"(In=@n4.On Next)
   n5 = VariableAccessor.Trigger(Variable=VariableMutator.Trigger(Variable=InitVariable.Trigger().Variable, Value=@n2.True).Variable)
   n6 = Value.Value:Float(Init Value=^MasterVolume, Target Value=@n3.Out, Set=@TriggerSequence2_0.Out 0, Reset=@n5.Value)
   n7 = VariableAccessor.Float(Variable=VariableMutator.Float(Variable=InitVariable.Float().Variable, Value=@n6.Output Value).Variable)
   WavePlayer_0 = UE.Wave Player:5dot1(Wave Asset=@ArrayRandomGet_0.Value, Pitch Shift=@n4.Value, Play=@n6.On Set)
   n8 = Value.Value:Float(Init Value=^MasterVolume, Target Value=@n3.Out, Set=@TriggerSequence2_0.Out 1, Reset=VariableAccessor.Trigger(Variable=@n5.Variable).Value)
   n9 = VariableAccessor.Float(Variable=VariableMutator.Float(Variable=InitVariable.Float().Variable, Value=@n8.Output Value).Variable)
   WavePlayer_1 = UE.Wave Player:5dot1(Wave Asset=@ArrayRandomGet_0.Value, Pitch Shift=@n4.Value, Play=@n8.On Set)
   Add_0 = UE.Add:Audio(
    PrimaryOperand=UE.Multiply:Audio by Float(AdditionalOperands=@n7.Value, PrimaryOperand=@WavePlayer_0.Out Side Right).Out
    AdditionalOperands=UE.Multiply:Audio by Float(AdditionalOperands=@n9.Value, PrimaryOperand=@WavePlayer_1.Out Side Right).Out
   )
   Add_1 = UE.Add:Audio(
    AdditionalOperands=UE.Multiply:Audio by Float(AdditionalOperands=@n9.Value, PrimaryOperand=@WavePlayer_1.Out Low Frequency).Out
    PrimaryOperand=UE.Multiply:Audio by Float(AdditionalOperands=@n7.Value, PrimaryOperand=@WavePlayer_0.Out Low Frequency).Out
   )
   Add_2 = UE.Add:Audio(
    PrimaryOperand=UE.Multiply:Audio by Float(AdditionalOperands=@n7.Value, PrimaryOperand=@WavePlayer_0.Out Side Left).Out
    AdditionalOperands=UE.Multiply:Audio by Float(AdditionalOperands=@n9.Value, PrimaryOperand=@WavePlayer_1.Out Side Left).Out
   )
   Add_3 = UE.Add:Audio(
    PrimaryOperand=UE.Multiply:Audio by Float(AdditionalOperands=@n7.Value, PrimaryOperand=@WavePlayer_0.Out Front Left).Out
    AdditionalOperands=UE.Multiply:Audio by Float(AdditionalOperands=@n9.Value, PrimaryOperand=@WavePlayer_1.Out Front Left).Out
   )
   Add_4 = UE.Add:Audio(
    PrimaryOperand=UE.Multiply:Audio by Float(AdditionalOperands=@n7.Value, PrimaryOperand=@WavePlayer_0.Out Front Right).Out
    AdditionalOperands=UE.Multiply:Audio by Float(AdditionalOperands=@n9.Value, PrimaryOperand=@WavePlayer_1.Out Front Right).Out
   )
   Add_5 = UE.Add:Audio(
    AdditionalOperands=UE.Multiply:Audio by Float(AdditionalOperands=@n9.Value, PrimaryOperand=@WavePlayer_1.Out Front Center).Out
    PrimaryOperand=UE.Multiply:Audio by Float(AdditionalOperands=@n7.Value, PrimaryOperand=@WavePlayer_0.Out Front Center).Out
   )
   TriggerControl_0 = UE.Trigger Control(Close=@WavePlayer_0.On Play, Trigger In=@WavePlayer_1.On Finished, Open=@WavePlayer_1.On Play)
   TriggerControl_1 = UE.Trigger Control(Trigger In=@WavePlayer_0.On Finished, Open=@WavePlayer_0.On Play, Close=@WavePlayer_1.On Play)
   TriggerAnyTriggerAny2_0 = "TriggerAny.Trigger Any (2)"(In 1=@TriggerControl_0.Trigger Out, In 0=@TriggerControl_1.Trigger Out)
  ]
  links [
   {dst="@UE.OutputFormat.5dot1.Audio:5", src=Add_0.Out}
   {dst="@UE.OutputFormat.5dot1.Audio:3", src=Add_1.Out}
   {dst="@UE.OutputFormat.5dot1.Audio:4", src=Add_2.Out}
   {dst="@UE.OutputFormat.5dot1.Audio:0", src=Add_3.Out}
   {dst="@UE.OutputFormat.5dot1.Audio:1", src=Add_4.Out}
   {dst="@UE.OutputFormat.5dot1.Audio:2", src=Add_5.Out}
   {dst="@UE.Source.OneShot.OnFinished", src=TriggerAnyTriggerAny2_0.Out}
  ]
 }]
properties {
 assetImportData=&self:AssetImportData
 numChannels=6
 outputFormat=FiveDotOne
 qualitySetting=Fort_Weapons
 sampleRate=48000
 soundClassObject=&Fort_WeaponRanged_Fire_1P_PSM
 concurrencyOverrides {maxCount=2, resolutionRule=StopOldest}
 concurrencySet [&Weapon_Gunfire_Local_Conc]
}
