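# One-shot stereo MetaSound that plays the overheat warning for the Date Gauntlet's primary fire.
# OnPlay triggers a single Wave Player on the warning wave, sending both channels straight to the outputs and signalling OnFinished when done.
# The Master Volume and ShotsLeft inputs are declared but not wired to anything.
# It uses gadget-activate attenuation, a non-critical in-game UI sound class, and a concurrency override that stops the oldest instance.
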MetaSoundSource /HighTower/Audio/Date/S31_Combo/MSS_Date_Gauntlet_Primary_Overheat_Warning
inputs [
 Master Volume {type=Float}
 ShotsLeft {type=Int32}
 UE.Source.OnPlay {type=Trigger}
]
interfaces [UE.OutputFormat.Stereo:1.0, UE.Source.OneShot:1.0, UE.Source:1.1]
outputs [
 "UE.OutputFormat.Stereo.Audio:0" {type=Audio}
 "UE.OutputFormat.Stereo.Audio:1" {type=Audio}
 UE.Source.OneShot.OnFinished {type=Trigger}
]
pages [Default {
  input_defaults {Master Volume=2, ShotsLeft=0, UE.Source.OnPlay=null}
  nodes [
   InitVariable.Float:Array()
   WavePlayer_0 = UE.Wave Player:Stereo(Play=^UE.Source.OnPlay) {also {defaults {Wave Asset=&/Date_Gauntlet_Primary_Fire_Overheat_Warning}}}
  ]
  links [
   {dst="@UE.Source.OneShot.OnFinished", src=WavePlayer_0.On Finished}
   {dst="@UE.OutputFormat.Stereo.Audio:0", src=WavePlayer_0.Out Left}
   {dst="@UE.OutputFormat.Stereo.Audio:1", src=WavePlayer_0.Out Right}
  ]
 }]
properties {
 assetImportData=&self:AssetImportData
 attenuationSettings=&Fort_GadgetActivate_Att
 bOverrideConcurrency=true
 maxDistance=4000
 numChannels=2
 outputFormat=Stereo
 sampleRate=48000
 soundClassObject=&Fort_UI_InGame_NonCritical
 volume=1.5
 concurrencyOverrides {maxCount=3, resolutionRule=StopOldest}
}
