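# One-shot stereo MetaSound for the Date Gauntlet primary fire overheat, using the first-person weapons sound class and weapon reload attenuation.
# On play, a random Overheat Start wave plays immediately and a delayed trigger plays a random Overheat End wave, both mixed under Master Volume.
# The default page leaves Overheat End empty, and the start player alone signals OnFinished, regardless of the delayed layer.
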MetaSoundSource /HighTower/Audio/Date/S31_Combo/MSS_Date_Gauntlet_Primary_Overheated
inputs [
 Master Volume {type=Float}
 Overheat End {type=WaveAsset:Array}
 Overheat Start {type=WaveAsset:Array}
 UE.Source.OnPlay {type=Trigger}
]
interfaces [UE.OutputFormat.Stereo:1.0, UE.Source.OneShot:1.0, UE.Source:1.1]
outputs [
 "UE.OutputFormat.Stereo.Audio:0" {type=Audio}
 "UE.OutputFormat.Stereo.Audio:1" {type=Audio}
 UE.Source.OneShot.OnFinished {type=Trigger}
]
pages [Default {
  input_defaults {
   Master Volume=1
   UE.Source.OnPlay=null
   Overheat End []
   Overheat Start [&/Date_Gauntlet_Primary_Fire_Overheat_01]
  }
  nodes [
   ArrayRandomGet_0 = Array.Random Get:WaveAsset:Array(In Array=^Overheat Start, Next=^UE.Source.OnPlay) {also {defaults {Weights []}}}
   WavePlayer_0 = UE.Wave Player:Stereo(Play=@ArrayRandomGet_0.On Next, Wave Asset=@ArrayRandomGet_0.Value)
   TriggerDelay_0 = UE.Trigger Delay(In=^UE.Source.OnPlay) {also {defaults {Delay Time=1.27}}}
   ArrayRandomGet_1 = Array.Random Get:WaveAsset:Array(In Array=^Overheat End, Next=@TriggerDelay_0.Out) {also {defaults {Weights []}}}
   WavePlayer_1 = UE.Wave Player:Stereo(Play=@ArrayRandomGet_1.On Next, Wave Asset=@ArrayRandomGet_1.Value)
   AudioMixerAudioMixerStereo2_0 = "AudioMixer.Audio Mixer (Stereo, 2)"(
    Gain 0=^Master Volume
    Gain 1=^Master Volume
    In 0 L=@WavePlayer_0.Out Left
    In 0 R=@WavePlayer_0.Out Right
    In 1 L=@WavePlayer_1.Out Left
    In 1 R=@WavePlayer_1.Out Right
   )
  ]
  links [
   {dst="@UE.OutputFormat.Stereo.Audio:0", src=AudioMixerAudioMixerStereo2_0.Out L}
   {dst="@UE.OutputFormat.Stereo.Audio:1", src=AudioMixerAudioMixerStereo2_0.Out R}
   {dst="@UE.Source.OneShot.OnFinished", src=WavePlayer_0.On Finished}
  ]
 }]
properties {
 assetImportData=&self:AssetImportData
 attenuationSettings=&Fort_Weapon_Reload_Att
 maxDistance=2250
 numChannels=2
 outputFormat=Stereo
 sampleRate=48000
 soundClassObject=/Game/Sounds/FortSoundClassesAndModes/Fort_Events_Weapons_1P
 volume=0.7
}
