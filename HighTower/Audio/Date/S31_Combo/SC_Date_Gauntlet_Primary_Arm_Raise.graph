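# Sound cue for the arm-raise foley of the Date Gauntlet primary fire, played as a weapon foley sound.
# A Random node picks one of four equally weighted arm-swing waves, and a Modulator applies slight pitch variation.
# It plays at reduced volume with weapon reload attenuation, under an owner-limited concurrency override that caps simultaneous instances.
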SoundCue /HighTower/Audio/Date/S31_Combo/SC_Date_Gauntlet_Primary_Arm_Raise
root=Modulator_1
nodes [
 {id=Modulator_1, children [Random_1], properties {pitchMax=1.1, pitchMin=0.9}}
 {id=Random_1, children [WavePlayer_0,WavePlayer_1,WavePlayer_2,WavePlayer_3], properties {weights [1,1,1,1]}}
 {id=WavePlayer_0, properties {soundWaveAssetPtr=&/Date_Gauntlet_Primary_Fire_Arm_Swing_01}}
 {id=WavePlayer_1, properties {soundWaveAssetPtr=&/Date_Gauntlet_Primary_Fire_Arm_Swing_02}}
 {id=WavePlayer_2, properties {soundWaveAssetPtr=&/Date_Gauntlet_Primary_Fire_Arm_Swing_03}}
 {id=WavePlayer_3, properties {soundWaveAssetPtr=&/Date_Gauntlet_Primary_Fire_Arm_Swing_04}}
]
properties {
 attenuationSettings=&Fort_Weapon_Reload_Att
 bOverrideConcurrency=true
 maxDistance=2250
 soundClassObject=&Fort_WeaponFoley
 volumeMultiplier=0.5
 concurrencyOverrides {bLimitToOwner=true, maxCount=5}
 concurrencySet [null]
}
