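# SoundCue for the close-range explosion of the Date Gauntlet's primary-fire projectile.
# A Random node picks one of four equally weighted close explosion waves, then a Modulator varies its pitch and volume slightly.
# It uses a dedicated projectile-impact attenuation and the medium explosives sound class.
# A concurrency override caps simultaneous instances and sets a short retrigger time.
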SoundCue /HighTower/Audio/Date/S31_Combo/SC_Date_Gauntlet_Primary_Proj_Expl
root=Modulator_1
nodes [
 {id=Modulator_1, children [Random_0], properties {pitchMax=1.1, pitchMin=0.9, volumeMax=1, volumeMin=0.7}}
 {id=Random_0, children [WavePlayer_9, WavePlayer_10, WavePlayer_11, WavePlayer_12], properties {weights [1,1,1,1]}}
 {id=WavePlayer_10, properties {soundWaveAssetPtr=&/Date_Gauntlet_Primary_Fire_Expl_Close_02}}
 {id=WavePlayer_11, properties {soundWaveAssetPtr=&/Date_Gauntlet_Primary_Fire_Expl_Close_03}}
 {id=WavePlayer_12, properties {soundWaveAssetPtr=&/Date_Gauntlet_Primary_Fire_Expl_Close_04}}
 {id=WavePlayer_9, properties {soundWaveAssetPtr=&/Date_Gauntlet_Primary_Fire_Expl_Close_01}}
]
properties {
 attenuationSettings=&/Att_Date_Gauntlet_Primary_Proj_Impact
 bOverrideConcurrency=true
 maxDistance=20100
 soundClassObject=/Game/Sounds/FortSoundClassesAndModes/Fort_Explosives_Medium
 volumeMultiplier=0.6
 concurrencyOverrides {maxCount=3, retriggerTime=0.1}
 concurrencySet [null]
}
