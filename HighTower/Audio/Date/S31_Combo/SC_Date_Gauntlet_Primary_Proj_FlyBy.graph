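# Sound cue for the projectile fly-by whoosh of the Date Gauntlet primary fire.
# A Random node picks one of five equally weighted fly-by waves, then feeds a Modulator at the root.
# It uses a dedicated whoosh-by attenuation asset and the medium explosives sound class, with reduced volume.
# A concurrency override caps simultaneous instances per owner.
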SoundCue /HighTower/Audio/Date/S31_Combo/SC_Date_Gauntlet_Primary_Proj_FlyBy
root=Modulator_1
nodes [
 {id=Modulator_1, children [Random_1]}
 {id=Random_1, children [WavePlayer_0,WavePlayer_1,WavePlayer_2,WavePlayer_3,WavePlayer_4], properties {weights [1,1,1,1,1]}}
 {id=WavePlayer_0, properties {soundWaveAssetPtr=&/Date_Gauntlet_Primary_Fire_Flyby_01}}
 {id=WavePlayer_1, properties {soundWaveAssetPtr=&/Date_Gauntlet_Primary_Fire_Flyby_02}}
 {id=WavePlayer_2, properties {soundWaveAssetPtr=&/Date_Gauntlet_Primary_Fire_Flyby_03}}
 {id=WavePlayer_3, properties {soundWaveAssetPtr=&/Date_Gauntlet_Primary_Fire_Flyby_04}}
 {id=WavePlayer_4, properties {soundWaveAssetPtr=&/Date_Gauntlet_Primary_Fire_Flyby_05}}
]
properties {
 attenuationSettings=&/Att_Date_Gauntlet_Primary_Proj_WhooshBy
 bOverrideConcurrency=true
 maxDistance=8200
 soundClassObject=/Game/Sounds/FortSoundClassesAndModes/Fort_Explosives_Medium
 volumeMultiplier=0.6
 concurrencyOverrides {bLimitToOwner=true, maxCount=5}
 concurrencySet [null]
}
