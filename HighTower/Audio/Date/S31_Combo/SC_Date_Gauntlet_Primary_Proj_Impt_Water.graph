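# Sound cue for the Date Gauntlet primary projectile impact on water, layering an explosion with water debris bursts.
# A half-volume, pitch-varied random gauntlet explosion mixes with random close water bursts and slightly delayed, pitch-varied distant bursts.
# Each layer has its own attenuation, and the close water layer overrides the cue's large-explosives sound class with a third-person ranged-fire class.
# A concurrency override caps simultaneous instances.
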SoundCue /HighTower/Audio/Date/S31_Combo/SC_Date_Gauntlet_Primary_Proj_Impt_Water
root=Mixer_3
nodes [
 {
  id=Attenuation_2
  children [Random_7]
  properties {attenuationSettings=/Game/Sounds/Attenuation/Fort_Explosion_Close_Athena_Att, attenuationOverrides {falloffDistance=1600}}
 }
 {
  id=Attenuation_3
  children [Modulator_7]
  properties {attenuationSettings=&/Att_Date_Gauntlet_Primary_Proj_Impact}
 }
 {id=Attenuation_4, children [Modulator_5], properties {attenuationSettings=/Game/Sounds/Attenuation/Fort_Explosion_Water_Distant_Athena_Att}}
 {id=Delay_5, children [Attenuation_4], properties {delayMax=0.15, delayMin=0.15}}
 {id=Mixer_2, children [SoundClass_4,Delay_5], properties {inputVolume [1,1]}}
 {id=Mixer_3, children [Attenuation_3, Mixer_2], properties {inputVolume [0.5,1]}}
 {id=Modulator_4, children [Attenuation_2]}
 {id=Modulator_5, children [Random_3], properties {pitchMax=1, pitchMin=0.85, volumeMax=1, volumeMin=1}}
 {id=Modulator_7, children [Random_2], properties {pitchMax=1.1, pitchMin=0.9}}
 {id=Random_2, children [WavePlayer_9, WavePlayer_10, WavePlayer_11, WavePlayer_7], properties {weights [1,1,1,1]}}
 {id=Random_3, children [WavePlayer_3, WavePlayer_13, WavePlayer_5], properties {bShouldExcludeFromBranchCulling=true, weights [1,1,1]}}
 {id=Random_7, children [WavePlayer_8, WavePlayer_23, WavePlayer_24, WavePlayer_25], properties {weights [1,1,1,1]}}
 {id=SoundClass_4, children [Modulator_4], properties {soundClassOverride=&Fort_WeaponRanged_Fire_3P_PSM}}
 {id=WavePlayer_10, properties {soundWaveAssetPtr=&/Date_Gauntlet_Primary_Fire_Expl_Close_02}}
 {id=WavePlayer_11, properties {soundWaveAssetPtr=&/Date_Gauntlet_Primary_Fire_Expl_Close_03}}
 {id=WavePlayer_13, properties {soundWaveAssetPtr=/Game/Athena/Sounds/Weapons/Explosions/EXPL_Debris_Water_Burst_Distant_02}}
 {id=WavePlayer_23, properties {soundWaveAssetPtr=/Game/Athena/Sounds/Weapons/Explosions/EXPL_Debris_Water_Burst_Close_03}}
 {id=WavePlayer_24, properties {soundWaveAssetPtr=/Game/Athena/Sounds/Weapons/Explosions/EXPL_Debris_Water_Burst_Close_02}}
 {id=WavePlayer_25, properties {soundWaveAssetPtr=/Game/Athena/Sounds/Weapons/Explosions/EXPL_Debris_Water_Burst_Close_01}}
 {id=WavePlayer_3, properties {soundWaveAssetPtr=/Game/Athena/Sounds/Weapons/Explosions/EXPL_Debris_Water_Burst_Distant_03}}
 {id=WavePlayer_5, properties {soundWaveAssetPtr=/Game/Athena/Sounds/Weapons/Explosions/EXPL_Debris_Water_Burst_Distant_01}}
 {id=WavePlayer_7, properties {soundWaveAssetPtr=&/Date_Gauntlet_Primary_Fire_Expl_Close_04}}
 {id=WavePlayer_8, properties {soundWaveAssetPtr=/Game/Athena/Sounds/Weapons/Explosions/EXPL_Debris_Water_Burst_Close_04}}
 {id=WavePlayer_9, properties {soundWaveAssetPtr=&/Date_Gauntlet_Primary_Fire_Expl_Close_01}}
]
properties {
 bOverrideConcurrency=true
 maxDistance=20100
 soundClassObject=/Game/Sounds/FortSoundClassesAndModes/Fort_Explosives_Large_PSM
 volumeMultiplier=0.7
 concurrencyOverrides {maxCount=3}
}
