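# Sound cue for the first-person secondary build sound of the Date Gauntlet, routed through the first-person ranged weapon fire sound class.
# A Random node picks one of three equally weighted close-perspective wave variations.
# A Modulator at the root gives the chosen wave a fixed upward pitch shift at unchanged volume, with no random range on either.
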SoundCue /HighTower/Audio/Date/S31_Combo/SC_Date_Gauntlet_Secondary_Build_1p
root=Modulator_5
nodes [
 {id=Modulator_5, children [Random_0], properties {pitchMax=1.1, pitchMin=1.1, volumeMax=1, volumeMin=1}}
 {id=Random_0, children [WavePlayer_0,WavePlayer_1,WavePlayer_2], properties {weights [1,1,1]}}
 {id=WavePlayer_0, properties {soundWaveAssetPtr=&/Date_Gauntlet_Secondary_Build_1P_Close_01}}
 {id=WavePlayer_1, properties {soundWaveAssetPtr=&/Date_Gauntlet_Secondary_Build_1P_Close_02}}
 {id=WavePlayer_2, properties {soundWaveAssetPtr=&/Date_Gauntlet_Secondary_Build_1P_Close_03}}
]
properties {maxDistance=2097152, soundClassObject=&Fort_WeaponRanged_Fire_1P_PSM}
