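# Third-person sound cue for the secondary "Build" sound of the Date gauntlet, routed to the ranged weapon fire 3P sound class.
# A mixer blends close, medium and distant layers, each with its own projectile attenuation, with the medium layer mixed lower.
# The close layer randomly picks one of three first-person waves at a fixed raised pitch.
# The medium and distant layers each play a single third-person wave through a modulator.
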SoundCue /HighTower/Audio/Date/S31_Combo/SC_Date_Gauntlet_Secondary_Build_3p
root=Mixer_2
nodes [
 {id=Attenuation_3, children [Modulator_0], properties {attenuationSettings=&Fort_Weapon_Projectile_Close_Athena_Att}}
 {
  id=Attenuation_4
  children [Modulator_4]
  properties {attenuationSettings=&Fort_Weapon_Projectile_Distant_Mono_Athena_Att}
 }
 {
  id=Attenuation_5
  children [Modulator_1]
  properties {attenuationSettings=&Fort_Weapon_Projectile_Medium_Mono_Athena_Att}
 }
 {id=Mixer_2, children [Attenuation_3, Attenuation_5, Attenuation_4], properties {inputVolume [1,0.4,1]}}
 {id=Modulator_0, children [Random_1], properties {pitchMax=1.1, pitchMin=1.1, volumeMax=1, volumeMin=1}}
 {id=Modulator_1, children [WavePlayer_10], properties {volumeMax=1, volumeMin=0.9}}
 {id=Modulator_4, children [WavePlayer_9]}
 {id=Random_1, children [WavePlayer_6,WavePlayer_7,WavePlayer_8], properties {weights [1,1,1]}}
 {id=WavePlayer_10, properties {soundWaveAssetPtr=&/Date_Gauntlet_Secondary_Build_3P_Medium_01}}
 {id=WavePlayer_6, properties {soundWaveAssetPtr=&/Date_Gauntlet_Secondary_Build_1P_Close_01}}
 {id=WavePlayer_7, properties {soundWaveAssetPtr=&/Date_Gauntlet_Secondary_Build_1P_Close_02}}
 {id=WavePlayer_8, properties {soundWaveAssetPtr=&/Date_Gauntlet_Secondary_Build_1P_Close_03}}
 {id=WavePlayer_9, properties {soundWaveAssetPtr=&/Date_Gauntlet_Secondary_Build_3P_Distant_01}}
]
properties {maxDistance=27500, soundClassObject=&Fort_WeaponRanged_Fire_3P_PSM}
