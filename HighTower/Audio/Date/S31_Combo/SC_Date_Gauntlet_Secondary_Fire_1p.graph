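# Sound cue for the first-person secondary fire of the Date Gauntlet weapon, assigned to the ranged weapon fire 1P sound class.
# A Random node picks with equal weight among three close-perspective fire wave variants.
# A Modulator at the root raises pitch by a fixed amount and leaves volume unchanged, so variation comes only from the wave choice.
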SoundCue /HighTower/Audio/Date/S31_Combo/SC_Date_Gauntlet_Secondary_Fire_1p
root=Modulator_0
nodes [
 {id=Modulator_0, children [Random_1], properties {pitchMax=1.1, pitchMin=1.1, volumeMax=1, volumeMin=1}}
 {id=Random_1, children [WavePlayer_4, WavePlayer_5, WavePlayer_11], properties {weights [1,1,1]}}
 {id=WavePlayer_11, properties {soundWaveAssetPtr=&/Date_Gauntlet_Secondary_Fire_1P_Close_01}}
 {id=WavePlayer_4, properties {soundWaveAssetPtr=&/Date_Gauntlet_Secondary_Fire_1P_Close_03}}
 {id=WavePlayer_5, properties {soundWaveAssetPtr=&/Date_Gauntlet_Secondary_Fire_1P_Close_02}}
]
properties {maxDistance=2097152, soundClassObject=&Fort_WeaponRanged_Fire_1P_PSM}
