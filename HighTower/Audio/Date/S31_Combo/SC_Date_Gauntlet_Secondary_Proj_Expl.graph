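# Sound cue for the explosion impact of the Date Gauntlet secondary projectile, classed as a large explosive.
# A mixer layers close, medium and distant branches, each picking one of three waves at random through its own explosion attenuation, with the close layer mixed lower.
# Modulators vary volume on every layer and pitch on close and medium, while the distant layer is pitched down by a fixed amount.
# A concurrency override caps simultaneous instances per owner.
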SoundCue /HighTower/Audio/Date/S31_Combo/SC_Date_Gauntlet_Secondary_Proj_Expl
root=Mixer_0
nodes [
 {id=Attenuation_0, children [Random_5], properties {attenuationSettings=/Game/Sounds/Attenuation/Fort_Explosion_Medium_Athena_Att}}
 {id=Attenuation_1, children [Random_6], properties {attenuationSettings=/Game/Sounds/Attenuation/Fort_Explosion_Distant_Athena_Att}}
 {
  id=Attenuation_7
  children [Modulator_3]
  properties {attenuationSettings=/Game/Sounds/Attenuation/Fort_Explosion_Close_Athena_Att, attenuationOverrides {falloffDistance=1600}}
 }
 {id=Mixer_0, children [Attenuation_7, Modulator_2, Modulator_0], properties {inputVolume [0.7,2,2]}}
 {id=Modulator_0, children [Attenuation_1], properties {pitchMax=0.7, pitchMin=0.7, volumeMax=1, volumeMin=0.9}}
 {id=Modulator_2, children [Attenuation_0], properties {pitchMax=1.1, pitchMin=0.9, volumeMax=1.1, volumeMin=0.9}}
 {id=Modulator_3, children [Random_4], properties {pitchMax=1.1, pitchMin=0.9, volumeMax=1.1, volumeMin=0.9}}
 {id=Random_4, children [WavePlayer_14, WavePlayer_15, WavePlayer_16], properties {weights [1,1,1]}}
 {id=Random_5, children [WavePlayer_17, WavePlayer_18, WavePlayer_19], properties {weights [1,1,1]}}
 {id=Random_6, children [WavePlayer_20, WavePlayer_21, WavePlayer_22], properties {weights [1,1,1]}}
 {id=WavePlayer_14, properties {soundWaveAssetPtr=&/Date_Gauntlet_Secondary_Proj_Impt_Close_01}}
 {id=WavePlayer_15, properties {soundWaveAssetPtr=&/Date_Gauntlet_Secondary_Proj_Impt_Close_02}}
 {id=WavePlayer_16, properties {soundWaveAssetPtr=&/Date_Gauntlet_Secondary_Proj_Impt_Close_03}}
 {id=WavePlayer_17, properties {soundWaveAssetPtr=&/Date_Gauntlet_Secondary_Proj_Impt_Med_01}}
 {id=WavePlayer_18, properties {soundWaveAssetPtr=&/Date_Gauntlet_Secondary_Proj_Impt_Med_02}}
 {id=WavePlayer_19, properties {soundWaveAssetPtr=&/Date_Gauntlet_Secondary_Proj_Impt_Med_03}}
 {id=WavePlayer_20, properties {soundWaveAssetPtr=&/Date_Gauntlet_Secondary_Proj_Impt_Dist_01}}
 {id=WavePlayer_21, properties {soundWaveAssetPtr=&/Date_Gauntlet_Secondary_Proj_Impt_Dist_02}}
 {id=WavePlayer_22, properties {soundWaveAssetPtr=&/Date_Gauntlet_Secondary_Proj_Impt_Dist_03}}
]
properties {
 bOverrideConcurrency=true
 maxDistance=27500
 soundClassObject=/Game/Sounds/FortSoundClassesAndModes/Fort_Explosives_Large_PSM
 volumeMultiplier=0.7
 concurrencyOverrides {bLimitToOwner=true, maxCount=4}
 concurrencySet [null]
}
