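# Sound cue for the whoosh fly-by of the Date Gauntlet's secondary projectile.
# It picks one of three equally weighted whoosh variants at random and passes it through a Modulator at default settings for variation.
# It reuses the primary projectile's whoosh-by attenuation and plays in the medium explosives sound class at reduced volume.
# A concurrency override caps simultaneous plays per owner.
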SoundCue /HighTower/Audio/Date/S31_Combo/SC_Date_Gauntlet_Secondary_Proj_FlyBy
root=Modulator_1
nodes [
 {id=Modulator_1, children [Random_0]}
 {id=Random_0, children [WavePlayer_5,WavePlayer_6,WavePlayer_7], properties {weights [1,1,1]}}
 {id=WavePlayer_5, properties {soundWaveAssetPtr=&/Date_Gauntlet_Secondary_Proj_Whoosh_01}}
 {id=WavePlayer_6, properties {soundWaveAssetPtr=&/Date_Gauntlet_Secondary_Proj_Whoosh_02}}
 {id=WavePlayer_7, properties {soundWaveAssetPtr=&/Date_Gauntlet_Secondary_Proj_Whoosh_03}}
]
properties {
 attenuationSettings=&/Att_Date_Gauntlet_Primary_Proj_WhooshBy
 bOverrideConcurrency=true
 maxDistance=8200
 soundClassObject=/Game/Sounds/FortSoundClassesAndModes/Fort_Explosives_Medium
 volumeMultiplier=0.7
 concurrencyOverrides {bLimitToOwner=true, maxCount=5}
 concurrencySet [null]
}
